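# Physics asset for the PeacefulPoem cape cosmetic, setting up rigid-body capsules and joints for the cape's dynamic bones.
# Center, left and right chains of seven capsules each hang from small master bodies linked to spine_05.
# Joints within each chain have tight swing and twist limits, with one swing axis looser at the third and fourth segments.
# Large pelvis and spine capsules form the torso, joined to each other and the masters with free angular motion.
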PhysicsAsset /BRCosmetics/Accessories/FORT_Capes/Cape_M_MED_PeacefulPoem/Meshes/Fortnite_Base_Cape_PeacefulPoem_Dyn
bodies {
 dyn_cape_01 [{shapeType=Capsule, length=7, radius=5, shapeName=dyn_cape_01_capsule, center=<8 0 8>, rotation=<rot -90 0 0>}]
 dyn_cape_02 [{shapeType=Capsule, length=7, radius=5, shapeName=dyn_cape_02_capsule, center=<8 0 5>, rotation=<rot -90 0 0>}]
 dyn_cape_03 [{shapeType=Capsule, length=7, radius=5, shapeName=dyn_cape_03_capsule, center=<8 0 5>, rotation=<rot -90 0 0>}]
 dyn_cape_04 [{shapeType=Capsule, length=7, radius=5, shapeName=dyn_cape_04_capsule, center=<8 0 5>, rotation=<rot -90 0 0>}]
 dyn_cape_05 [{shapeType=Capsule, length=7, radius=5, shapeName=dyn_cape_05_capsule, center=<8 0 5>, rotation=<rot -90 0 0>}]
 dyn_cape_06 [{shapeType=Capsule, length=7, radius=5, shapeName=dyn_cape_06_capsule, center=<8 0 5>, rotation=<rot -90 0 0>}]
 dyn_cape_07 [{shapeType=Capsule, length=7, radius=5, shapeName=dyn_cape_07_capsule, center=<8 0 5>, rotation=<rot -90 0 0>}]
 dyn_cape_MSTR [{shapeType=Capsule, length=0.505, radius=0.505, shapeName=dyn_cape_MSTR_capsule, rotation=<rot 0 -90 0>}]
 dyn_cape_l_01 [{shapeType=Capsule, length=7, radius=5, shapeName=dyn_cape_l_01_capsule, center=<8 0 8>, rotation=<rot -90 0 0>}]
 dyn_cape_l_02 [{shapeType=Capsule, length=7, radius=5, shapeName=dyn_cape_l_02_capsule, center=<8 0 5>, rotation=<rot -90 0 0>}]
 dyn_cape_l_03 [{shapeType=Capsule, length=7, radius=5, shapeName=dyn_cape_l_03_capsule, center=<8 0 5>, rotation=<rot -90 0 0>}]
 dyn_cape_l_04 [{shapeType=Capsule, length=7, radius=5, shapeName=dyn_cape_l_04_capsule, center=<8 0 5>, rotation=<rot -90 0 0>}]
 dyn_cape_l_05 [{shapeType=Capsule, length=7, radius=5, shapeName=dyn_cape_l_05_capsule, center=<8 0 5>, rotation=<rot -90 0 0>}]
 dyn_cape_l_06 [{shapeType=Capsule, length=7, radius=5, shapeName=dyn_cape_l_06_capsule, center=<8 0 5>, rotation=<rot -90 0 0>}]
 dyn_cape_l_07 [{shapeType=Capsule, length=7, radius=5, shapeName=dyn_cape_l_07_capsule, center=<8 0 5>, rotation=<rot -90 0 0>}]
 dyn_cape_l_MSTR [{shapeType=Capsule, length=0.505, radius=0.505, shapeName=dyn_cape_l_MSTR_capsule, rotation=<rot 0 -90 0>}]
 dyn_cape_r_01 [{shapeType=Capsule, length=7, radius=5, shapeName=dyn_cape_r_01_capsule, center=<8 0 8>, rotation=<rot -90 0 0>}]
 dyn_cape_r_02 [{shapeType=Capsule, length=7, radius=5, shapeName=dyn_cape_r_02_capsule, center=<8 0 5>, rotation=<rot -90 0 0>}]
 dyn_cape_r_03 [{shapeType=Capsule, length=7, radius=5, shapeName=dyn_cape_r_03_capsule, center=<8 0 5>, rotation=<rot -90 0 0>}]
 dyn_cape_r_04 [{shapeType=Capsule, length=7, radius=5, shapeName=dyn_cape_r_04_capsule, center=<8 0 5>, rotation=<rot -90 0 0>}]
 dyn_cape_r_05 [{shapeType=Capsule, length=7, radius=5, shapeName=dyn_cape_r_05_capsule, center=<8 0 5>, rotation=<rot -90 0 0>}]
 dyn_cape_r_06 [{shapeType=Capsule, length=7, radius=5, shapeName=dyn_cape_r_06_capsule, center=<8 0 5>, rotation=<rot -90 0 0>}]
 dyn_cape_r_07 [{shapeType=Capsule, length=7, radius=5, shapeName=dyn_cape_r_07_capsule, center=<8 0 5>, rotation=<rot -90 0 0>}]
 dyn_cape_r_MSTR [{shapeType=Capsule, length=0.505, radius=0.505, shapeName=dyn_cape_r_MSTR_capsule, rotation=<rot 0 -90 0>}]
 pelvis [{shapeType=Capsule, length=12, radius=15, shapeName=pelvis_capsule, center=<-10.642 -8.5081 0>}]
 spine_01 [{shapeType=Capsule, length=12, radius=15, shapeName=spine_01_capsule, center=<0 -6.3842 0>}]
 spine_02 [{shapeType=Capsule, length=12, radius=15, shapeName=spine_02_capsule, center=<0 -5.3104 0>}]
 spine_03 [{shapeType=Capsule, length=12, radius=15, shapeName=spine_03_capsule, center=<0 -4.1806 0>}]
 spine_04 [{shapeType=Capsule, length=12, radius=15, shapeName=spine_04_capsule, center=<0 -2.9416 0>}]
 spine_05 [{shapeType=Capsule, length=12, radius=15, shapeName=spine_05_capsule, center=<0 -3.2049 0>}]
}
constraints [
 {bone1Name=dyn_cape_01, bone2Name=dyn_cape_MSTR, swing1LimitDegrees=8, swing2LimitDegrees=15, twistLimitDegrees=5}
 {bone1Name=dyn_cape_02, bone2Name=dyn_cape_01, swing1LimitDegrees=8, swing2LimitDegrees=15, twistLimitDegrees=5}
 {bone1Name=dyn_cape_03, bone2Name=dyn_cape_02, swing1LimitDegrees=20, swing2LimitDegrees=15, twistLimitDegrees=5}
 {bone1Name=dyn_cape_04, bone2Name=dyn_cape_03, swing1LimitDegrees=20, swing2LimitDegrees=15, twistLimitDegrees=5}
 {bone1Name=dyn_cape_05, bone2Name=dyn_cape_04, swing1LimitDegrees=8, swing2LimitDegrees=15, twistLimitDegrees=5}
 {bone1Name=dyn_cape_06, bone2Name=dyn_cape_05, swing1LimitDegrees=8, swing2LimitDegrees=15, twistLimitDegrees=5}
 {bone1Name=dyn_cape_07, bone2Name=dyn_cape_06, swing1LimitDegrees=8, swing2LimitDegrees=15, twistLimitDegrees=5}
 {bone1Name=dyn_cape_l_01, bone2Name=dyn_cape_l_MSTR, swing1LimitDegrees=8, swing2LimitDegrees=15, twistLimitDegrees=5}
 {bone1Name=dyn_cape_l_02, bone2Name=dyn_cape_l_01, swing1LimitDegrees=8, swing2LimitDegrees=15, twistLimitDegrees=5}
 {bone1Name=dyn_cape_l_03, bone2Name=dyn_cape_l_02, swing1LimitDegrees=20, swing2LimitDegrees=15, twistLimitDegrees=5}
 {bone1Name=dyn_cape_l_04, bone2Name=dyn_cape_l_03, swing1LimitDegrees=20, swing2LimitDegrees=15, twistLimitDegrees=5}
 {bone1Name=dyn_cape_l_05, bone2Name=dyn_cape_l_04, swing1LimitDegrees=8, swing2LimitDegrees=15, twistLimitDegrees=5}
 {bone1Name=dyn_cape_l_06, bone2Name=dyn_cape_l_05, swing1LimitDegrees=8, swing2LimitDegrees=15, twistLimitDegrees=5}
 {bone1Name=dyn_cape_l_07, bone2Name=dyn_cape_l_06, swing1LimitDegrees=8, swing2LimitDegrees=15, twistLimitDegrees=5}
 {bone1Name=dyn_cape_r_01, bone2Name=dyn_cape_r_MSTR, swing1LimitDegrees=8, swing2LimitDegrees=15, twistLimitDegrees=5}
 {bone1Name=dyn_cape_r_02, bone2Name=dyn_cape_r_01, swing1LimitDegrees=8, swing2LimitDegrees=15, twistLimitDegrees=5}
 {bone1Name=dyn_cape_r_03, bone2Name=dyn_cape_r_02, swing1LimitDegrees=20, swing2LimitDegrees=15, twistLimitDegrees=5}
 {bone1Name=dyn_cape_r_04, bone2Name=dyn_cape_r_03, swing1LimitDegrees=20, swing2LimitDegrees=15, twistLimitDegrees=5}
 {bone1Name=dyn_cape_r_05, bone2Name=dyn_cape_r_04, swing1LimitDegrees=8, swing2LimitDegrees=15, twistLimitDegrees=5}
 {bone1Name=dyn_cape_r_06, bone2Name=dyn_cape_r_05, swing1LimitDegrees=8, swing2LimitDegrees=15, twistLimitDegrees=5}
 {bone1Name=dyn_cape_r_07, bone2Name=dyn_cape_r_06, swing1LimitDegrees=8, swing2LimitDegrees=15, twistLimitDegrees=5}
 {
  bone1Name=dyn_cape_MSTR
  bone2Name=spine_05
  swing1LimitDegrees=45
  swing1Motion=Free
  swing2LimitDegrees=45
  swing2Motion=Free
  twistLimitDegrees=45
  twistMotion=Free
 }
 {
  bone1Name=dyn_cape_r_MSTR
  bone2Name=spine_05
  swing1LimitDegrees=45
  swing1Motion=Free
  swing2LimitDegrees=45
  swing2Motion=Free
  twistLimitDegrees=45
  twistMotion=Free
 }
 {
  bone1Name=dyn_cape_l_MSTR
  bone2Name=spine_05
  swing1LimitDegrees=45
  swing1Motion=Free
  swing2LimitDegrees=45
  swing2Motion=Free
  twistLimitDegrees=45
  twistMotion=Free
 }
 {
  bone1Name=spine_05
  bone2Name=spine_04
  swing1LimitDegrees=45
  swing1Motion=Free
  swing2LimitDegrees=45
  swing2Motion=Free
  twistLimitDegrees=45
  twistMotion=Free
 }
 {
  bone1Name=spine_04
  bone2Name=spine_03
  swing1LimitDegrees=45
  swing1Motion=Free
  swing2LimitDegrees=45
  swing2Motion=Free
  twistLimitDegrees=45
  twistMotion=Free
 }
 {
  bone1Name=spine_03
  bone2Name=spine_02
  swing1LimitDegrees=45
  swing1Motion=Free
  swing2LimitDegrees=45
  swing2Motion=Free
  twistLimitDegrees=45
  twistMotion=Free
 }
 {
  bone1Name=spine_02
  bone2Name=spine_01
  swing1LimitDegrees=45
  swing1Motion=Free
  swing2LimitDegrees=45
  swing2Motion=Free
  twistLimitDegrees=45
  twistMotion=Free
 }
 {
  bone1Name=spine_01
  bone2Name=pelvis
  swing1LimitDegrees=45
  swing1Motion=Free
  swing2LimitDegrees=45
  swing2Motion=Free
  twistLimitDegrees=45
  twistMotion=Free
 }
]
